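# Foliage material function that reshapes a material's OpacityMask by distance, behind a static switch.
# It fills the mask solid within a soft, object-scaled sphere around LocationCore, with radius blending from near to far values by camera distance.
# The result is multiplied by a camera depth fade, broken up through HeightLerp with the HeightModulation input.
# Shadow and ray-traced paths keep the original mask.
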MaterialFunction /Game/Environments/Apollo/Foliage/MaterialFunctions/MF_DistanceMaskFade
graph {nodes [
  n1 = FunctionInput() {inputType=FunctionInput_MaterialAttributes}
  n2 = GetMaterialAttributes(None=@n1) {attributeGetTypes [#1]}
  n3 = TransformPosition(None=VectorParameter() {parameterName=LocationCore, defaultValue=<rgba 0 0 0 1>}.RGB) {transformType=TRANSFORMPOSSOURCE_World}
  n4 = Distance(A=WorldPosition().XYZ, B=@n3)
  n5 = CameraPositionWS()
  n6 = Divide(
   A=Subtract(A=Distance(A=@n3, B=@n5), B=ScalarParameter() {defaultValue=600, parameterName=DistanceMask_CloseFadeDistOffset})
   B=ScalarParameter() {defaultValue=80, parameterName=DistanceMask_CloseFadeDist}
  )
  n7 = MaterialFunctionCall() {materialFunction=/Engine/Functions/Engine_MaterialFunctions02/WorldPositionOffset/ObjectScale}
  n8 = Min(A=Min(A=@n7.Scale X, B=@n7.Scale X), B=@n7.Scale Z)
  n9 = Multiply(
   A=LinearInterpolate(
    A=ScalarParameter() {parameterName=DistanceMaskFade_CoreRadius_Near}
    B=ScalarParameter() {defaultValue=300.4869, parameterName=DistanceMaskFade_CoreRadius}
    Alpha=Saturate(None=@n6)
   )
   B=@n8
  )
  n10 = ScalarParameter() {defaultValue=16, parameterName=DistanceMaskFade_Core_Blend}
  n11 = Saturate(None=Divide(A=Subtract(A=@n4, B=@n9), B=Multiply(A=@n10, B=@n8)))
  n12 = Saturate(None=Add(A=@n2.OpacityMask, B=OneMinus(None=@n11)))
  n13 = MaterialFunctionCall(
   Transition Phase=MaterialFunctionCall(
    Fade Length=ScalarParameter() {defaultValue=200, group=Opacity Mask Fade, parameterName=OpacityMask_Fade_length}
    Fade Offset=ScalarParameter() {group=Opacity Mask Fade, parameterName=OpacityMask_Fade_offset}
    For Vertex Shader=StaticBool()
   ) {materialFunction=&CameraDepthFade}.Result
   Height Texture=Multiply(
    A=FunctionInput() {inputName=HeightModulation, inputType=FunctionInput_Scalar, sortPriority=4}
    B=ScalarParameter() {defaultValue=0.5, group=Opacity Mask Fade, parameterName=OpacityMask_Fade_Height_intensity}
   )
  ) {materialFunction=&HeightLerp}
  n14 = RayTracingQualitySwitch(Normal=ShadowReplace(Default=Multiply(A=@n12, B=@n13.Alpha), Shadow=@n2.OpacityMask), RayTraced=@n2.OpacityMask)
  FunctionOutput(None=StaticSwitchParameter(
    True=SetMaterialAttributes!(MaterialAttributes=@n2.MaterialAttributes, OpacityMask=@n14) {attributeSetTypes [#1]}
    False=@n1
   ) {parameterName=Use DistanceOpacityMaskFade})
 ]}
