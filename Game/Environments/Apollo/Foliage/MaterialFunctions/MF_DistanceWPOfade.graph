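# Foliage material function that distance-fades World Position Offset, blending between a collapse to the object pivot and the incoming WPO.
# Non-Nanite rendering drives the blend from object-to-camera distance with offset and length parameters.
# Nanite instead uses MF_WPODisableBlend, fed the fade length.
# A Mask input restores the original WPO, and the "Use WPO Fade and Shrink" static switch bypasses the effect entirely.
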MaterialFunction /Game/Environments/Apollo/Foliage/MaterialFunctions/MF_DistanceWPOfade
graph {nodes [
  n1 = FunctionInput() {inputType=FunctionInput_MaterialAttributes}
  n2 = GetMaterialAttributes(None=@n1) {attributeGetTypes [#1]}
  n3 = ObjectPositionWS()
  n4 = WorldPosition()
  n5 = Distance(A=ObjectPositionWS(), B=CameraPositionWS())
  n6 = ScalarParameter() {defaultValue=5000, group=WPO Fade and Shrink, parameterName=WPOFade_offset}
  n7 = ScalarParameter() {defaultValue=2500, group=WPO Fade and Shrink, parameterName=WPOFade_length}
  n8 = NaniteReplace!(
   Default=Divide(A=Add(A=@n5, B=@n6), B=@n7)
   Nanite=MaterialFunctionCall!(FadeBlendDistance=@n7, AutoHandleBadDistance) {materialFunction=/Game/Global/MaterialFunctions/TechArt/MF_WPODisableBlend}.VertexShaderOutput
  )
  n9 = LinearInterpolate(
   A=LinearInterpolate(A=Subtract(A=@n3, B=@n4.XYZ), B=@n2.WorldPositionOffset, Alpha=Saturate(None=@n8))
   B=@n2.WorldPositionOffset
   Alpha=FunctionInput() {inputName=Mask, inputType=FunctionInput_Scalar, sortPriority=4}
  )
  FunctionOutput(None=StaticSwitchParameter(
    True=SetMaterialAttributes!(MaterialAttributes=@n2.MaterialAttributes, WorldPositionOffset=@n9) {attributeSetTypes [#1]}
    False=@n1
   ) {parameterName=Use WPO Fade and Shrink})
 ]}
