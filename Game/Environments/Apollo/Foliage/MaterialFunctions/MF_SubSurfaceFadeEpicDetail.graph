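# Material function for foliage leaves that fades subsurface opacity with distance at Epic quality.
# It takes Material Attributes and lerps their Opacity toward a distant-value parameter, using camera-to-object-position distance remapped by offset and fade-distance parameters, then saturated.
# A Quality Switch applies the modified Opacity only at Epic, while the default path passes the input through unchanged.
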MaterialFunction /Game/Environments/Apollo/Foliage/MaterialFunctions/MF_SubSurfaceFadeEpicDetail
graph {nodes [
  n1 = FunctionInput() {inputType=FunctionInput_MaterialAttributes}
  n2 = GetMaterialAttributes(None=@n1) {attributeGetTypes [#1]}
  n3 = Add(
   A=Distance(A=CameraPositionWS(), B=ObjectPositionWS())
   B=ScalarParameter() {defaultValue=-10000, group=Shading Leafs EPIC Detail, parameterName=SubSurface_SSS_Opacity_FadeOffset}
  )
  n4 = ScalarParameter() {defaultValue=10000, group=Shading Leafs EPIC Detail, parameterName=SubSurface_SSS_Opacity_FadeDistance}
  n5 = LinearInterpolate(
   A=@n2.Opacity
   B=ScalarParameter() {defaultValue=0.5, group=Shading Leafs EPIC Detail, parameterName=SubSurface_SSS_Opacity_Distant}
   Alpha=Saturate(None=Divide(A=@n3, B=@n4))
  )
  FunctionOutput(None=QualitySwitch(
    Default=@n1
    Epic=SetMaterialAttributes!(MaterialAttributes=@n2.MaterialAttributes, Opacity=@n5) {attributeSetTypes [#1]}
   ))
 ]}
