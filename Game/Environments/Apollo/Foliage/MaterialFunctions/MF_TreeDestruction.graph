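# Material function for foliage that animates a tree falling, sinking and losing its leaves, enabled by the IsDestructionMaterial static switch.
# A looping Time-driven phase, offset by each point's distance from the actor origin relative to object radius, feeds separate power curves.
# These drive World Position Offset rotation about the actor pivot plus Z sinking, and texture-masked Opacity Mask erosion.
# The LeafErosionMask input is unconnected.
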MaterialFunction /Game/Environments/Apollo/Foliage/MaterialFunctions/MF_TreeDestruction
graph {nodes [
  FunctionInput() {inputName=LeafErosionMask, inputType=FunctionInput_Scalar, sortPriority=2}
  n1 = FunctionInput() {inputType=FunctionInput_MaterialAttributes, sortPriority=-1}
  n2 = GetMaterialAttributes(None=@n1) {attributeGetTypes [#1, #2]}
  n3 = VectorParameter() {parameterName=Falling_RotationAxis, defaultValue=<rgba 1 0 0 1>}
  n4 = WorldPosition()
  n5 = ActorPositionWS()
  n6 = ObjectRadius()
  n7 = Power(
   Base=Divide(A=Distance(A=@n4.XYZ, B=@n5), B=Multiply(A=@n6) {constB=2})
   Exp=ScalarParameter() {defaultValue=1, parameterName=DelayDestructionFallof}
  )
  n8 = Subtract(
   A=Multiply(A=Time(), B=ScalarParameter() {defaultValue=0.3643, parameterName=FallingSpeed})
   B=LinearInterpolate(A=ScalarParameter() {defaultValue=-0.05, parameterName=DelayDestruction}, Alpha=Saturate(None=@n7)) {constB=0}
  )
  n9 = Saturate(None=Multiply(A=ScalarParameter() {defaultValue=1.5, parameterName=Pause}, B=Frac(None=@n8)))
  n10 = Saturate(None=Power(Base=@n9, Exp=ScalarParameter() {defaultValue=4, parameterName=FallingAnimExponent}))
  n11 = Multiply(A=ScalarParameter() {defaultValue=0.1, parameterName=FallingRotationAngle}, B=@n10)
  n12 = Add(A=@n2.WorldPositionOffset, B=WorldPosition().XYZ)
  n13 = Power(Base=@n9, Exp=ScalarParameter() {defaultValue=8, parameterName=FallingSinkAnimExponent})
  n14 = ScalarParameter() {defaultValue=-2000, parameterName=FallingGroungSinkEnd}
  n15 = Multiply(A=Constant3Vector() {constant=<rgba 0 0 1 0>}, B=Add(A=ScalarParameter() {parameterName=FallingGroungSink}, B=Multiply(A=@n13, B=@n14)))
  n16 = Add(
   A=@n2.WorldPositionOffset
   B=Add(A=RotateAboutAxis(NormalizedRotationAxis=@n3.RGB, RotationAngle=@n11, PivotPoint=@n5, Position=@n12), B=Multiply(A=@n15, B=@n10))
  )
  n17 = TextureSampleParameter2D() {parameterName=Mask, texture=&T_WhitesRGB}
  n18 = ScalarParameter() {defaultValue=-2.5, parameterName=SubtractOpcaityDestruction}
  n19 = Power(Base=@n9, Exp=ScalarParameter() {defaultValue=1.5, parameterName=FallingLeafAnimExponent})
  n20 = Add(A=@n2.OpacityMask, B=LinearInterpolate(A=Constant(), B=Multiply(A=@n17.G, B=@n18), Alpha=Saturate(None=@n19)))
  FunctionOutput(None=StaticSwitchParameter(
    True=SetMaterialAttributes!(MaterialAttributes=@n2.MaterialAttributes, WorldPositionOffset=@n16, OpacityMask=@n20) {attributeSetTypes [#1, #2]}
    False=@n1
   ) {parameterName=IsDestructionMaterial})
 ]}
