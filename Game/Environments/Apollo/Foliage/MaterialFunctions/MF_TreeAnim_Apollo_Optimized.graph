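# Optimized Apollo foliage material function that adds wind-driven tree animation to material attributes through World Position Offset.
# Pivot and axis textures drive layered trunk, branch and leaf rotations, timed by a sine-wave texture.
# Wind direction, gusts and storm mask scale motion, with physics field forces added at Epic quality.
# Switches trim levels or select simple animation on mobile, before tree destruction and optional mobile near-camera leaf shrinking.
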MaterialFunction /Game/Environments/Apollo/Foliage/MaterialFunctions/MF_TreeAnim_Apollo_Optimized
graph {nodes [
  Constant3Vector()
  n1 = ComponentMask(None=NamedRerouteUsage()) {b=true, g=true, r=true}
  n2 = Add(A=ComponentMask(None=@n1) {b=true}, B=ComponentMask(None=@n1) {g=true})
  n3 = Add(A=@n2, B=ComponentMask(None=@n1) {r=true})
  FunctionOutput(None=Frac(None=Multiply(A=@n3) {constB=5.6843})) {outputName=Random Value perleaf, sortPriority=3}
  n4 = Constant3Vector()
  n5 = Constant3Vector()
  n6 = Multiply(A=NamedRerouteUsage(), B=NamedRerouteUsage())
  n7 = Multiply(A=NamedRerouteUsage(), B=NamedRerouteUsage())
  n8 = Add(A=Multiply(A=NamedRerouteUsage(), B=NamedRerouteUsage()), B=@n7)
  n9 = StaticSwitchParameter(
   True=QualitySwitch(Default=StaticSwitchParameter(True=@n7, False=@n8) {group=Optimziation, parameterName=Optimization_Disable_BranchAnim}, Low=@n7, Medium=@n7)
   False=@n7
  ) {parameterName=UseBranchLevel}
  n10 = StaticSwitchParameter(True=@n6, False=Add(A=@n6, B=@n9)) {group=Trunk, parameterName=IsTrunk}
  n11 = QualitySwitch(Default=@n9, Epic=@n10)
  n12 = QualitySwitch(
   Default=StaticSwitchParameter(True=StaticSwitchParameter(True=@n5, False=@n11) {group=Trunk, parameterName=IsTrunk}, False=@n10) {group=Optimziation, parameterName=Optimization_Disable_TrunkAnim}
   Low=StaticSwitchParameter(True=@n5, False=@n7) {group=Trunk, parameterName=IsTrunk}
  )
  n13 = FunctionInput(Preview=Constant()) {bUsePreviewValueAsDefault=true, inputName=OpaqueCanopyMask, sortPriority=343}
  n14 = StaticSwitchParameter(
   True=NamedRerouteUsage()
   False=StaticSwitchParameter(True=LinearInterpolate(A=@n4, B=@n12, Alpha=@n13), False=@n12) {parameterName=UseOpaqueCanopy}
  ) {parameterName=UseSimpleNonePivotPainterAnimation}
  n15 = NamedRerouteUsage()
  n16 = MaterialFunctionCall!(
   In=MaterialFunctionCall!(
    Material=FunctionInput() {inputType=FunctionInput_MaterialAttributes}
    WorldPositionOffset=ShadingPathSwitch(Default=@n14, Deferred=@n14, Forward=@n15, Mobile=@n15) {desc=Always simple animation on mobile and switch}
   ) {materialFunction=/Engine/Functions/MaterialLayerFunctions/MatLayerBlend_AddWorldPositionOffset}.Blended Material
   LeafErosionMask
  ) {materialFunction=&/MF_TreeDestruction}
  n17 = GetMaterialAttributes(None=@n16.Result) {attributeGetTypes [#1]}
  n18 = StaticSwitchParameter(True=NamedRerouteUsage(), False=NamedRerouteUsage()) {parameterName=UseBranchLevel}
  n19 = Subtract(
   A=@n18
   B=Add(
    A=WorldPosition().XYZ
    B=GetMaterialAttributes(None=@n16.Result) {attributeGetTypes [#1]}.WorldPositionOffset
   )
  )
  n20 = MaterialFunctionCall() {materialFunction=/Engine/Functions/Engine_MaterialFunctions02/WorldPositionOffset/ObjectScale}
  n21 = Subtract(
   A=Distance(A=@n18, B=CameraPositionWS())
   B=Multiply(A=@n20.Scale X, B=ScalarParameter() {group=NearLeafShrink, parameterName=NearLeafShrink_Offset})
  )
  n22 = Multiply(A=ScalarParameter() {defaultValue=325.7803, group=NearLeafShrink, parameterName=NearLeafShrink_Distance}, B=@n20.Scale X)
  n23 = OneMinus(None=Saturate(None=Divide(A=@n21, B=@n22)))
  n24 = SetMaterialAttributes!(MaterialAttributes=@n17.MaterialAttributes, WorldPositionOffset=Add(A=@n17.WorldPositionOffset, B=Multiply(A=@n19, B=@n23))) {attributeSetTypes [#1]}
  FunctionOutput(None=StaticSwitchParameter(True=ShadingPathSwitch(Default=@n16.Result, Mobile=@n24) {desc=Migrated from FeatureLevelSwitch}, False=@n16.Result) {parameterName=UseNearLeafShrinkonMobile})
  n25 = Add(A=Constant2Vector() {r=-0.5}, B=TextureCoordinate() {coordinateIndex=1})
  n26 = TextureObjectParameter() {parameterName=Position, samplerType=SAMPLERTYPE_LinearColor, texture=/Game/Environments/Apollo/Foliage/SpeedTree/PivotPainttest/Position}
  n27 = TextureProperty(Tex=@n26)
  n28 = AppendVector(A=Divide(A=ComponentMask(None=@n25) {r=true}, B=ComponentMask(None=@n27) {r=true}), B=ComponentMask(None=@n25) {g=true})
  NamedRerouteDeclaration(None=@n28) {name=UVsAnimTexture, nodeColor=<rgba 0 0.7187 1 1>}
  n29 = Constant() {r=2}
  n30 = NamedRerouteUsage()
  n31 = Subtract(A=NamedRerouteUsage(), B=AppendVector(A=Constant(), B=Divide(A=@n29, B=@n30)))
  n32 = TextureSampleParameter2D(UVs=@n31) {
   parameterName=Position
   samplerSource=SSM_Wrap_WorldGroupSettings
   samplerType=SAMPLERTYPE_LinearColor
   texture=&Position
  }
  NamedRerouteDeclaration(None=@n32.A) {name=Mask_Level_3, nodeColor=<rgba 0 0.4375 1 1>}
  n33 = Constant()
  n34 = NamedRerouteUsage()
  n35 = Subtract(A=NamedRerouteUsage(), B=AppendVector(A=Constant(), B=Divide(A=@n33, B=@n34)))
  n36 = TextureSampleParameter2D(UVs=@n35) {
   parameterName=X-Axis
   samplerSource=SSM_Wrap_WorldGroupSettings
   samplerType=SAMPLERTYPE_LinearColor
   texture=/Game/Environments/Apollo/Foliage/SpeedTree/PivotPainttest/XVector
  }
  n37 = Subtract(A=Multiply(A=@n36.RGB) {constB=2})
  n38 = AppendVector(A=Normalize(VectorInput=Transform(None=@n37) {transformSourceType=TRANSFORMSOURCE_Local}), B=Multiply(A=@n36.A) {constB=2048})
  NamedRerouteDeclaration(None=@n38) {name=XAxis_Level_1, nodeColor=<rgba 0 0.25 1 1>}
  n39 = Constant() {r=1}
  n40 = NamedRerouteUsage()
  n41 = Subtract(A=NamedRerouteUsage(), B=AppendVector(A=Constant(), B=Divide(A=@n39, B=@n40)))
  n42 = TextureSampleParameter2D(UVs=@n41) {
   parameterName=X-Axis
   samplerSource=SSM_Wrap_WorldGroupSettings
   samplerType=SAMPLERTYPE_LinearColor
   texture=&XVector
  }
  n43 = Subtract(A=Multiply(A=@n42.RGB) {constB=2})
  n44 = AppendVector(A=Normalize(VectorInput=Transform(None=@n43) {transformSourceType=TRANSFORMSOURCE_Local}), B=Multiply(A=@n42.A) {constB=2048})
  NamedRerouteDeclaration(None=@n44) {name=XAxis_Level2, nodeColor=<rgba 1 0.6094 0 1>}
  n45 = TextureSampleParameter2D(UVs=@n31) {
   parameterName=X-Axis
   samplerSource=SSM_Wrap_WorldGroupSettings
   samplerType=SAMPLERTYPE_LinearColor
   texture=&XVector
  }
  n46 = Subtract(A=Multiply(A=@n45.RGB) {constB=2})
  n47 = AppendVector(A=Normalize(VectorInput=Transform(None=@n46) {transformSourceType=TRANSFORMSOURCE_Local}), B=Multiply(A=@n45.A) {constB=2048})
  NamedRerouteDeclaration(None=@n47) {name=XAxis_Level_3, nodeColor=<rgba 1 0.3984 0 1>}
  n48 = Clamp(None=NamedRerouteUsage())
  NamedRerouteDeclaration(None=@n48) {name=Mask_Branches, nodeColor=<rgba 0.5234 0 1 1>}
  n49 = Multiply(A=Time(), B=ScalarParameter() {defaultValue=0.75, group=Anim_Speed_Control, parameterName=Speed_Master})
  NamedRerouteDeclaration(None=@n49) {name=Time_Master_Speed, nodeColor=<rgba 1 0.8672 0 1>}
  n50 = StaticSwitchParameter(True=NamedRerouteUsage(), False=NamedRerouteUsage()) {parameterName=UseBranchLevel}
  NamedRerouteDeclaration(None=@n50) {name=Branch_Level_Pivot, nodeColor=<rgba 0.4062 1 0 1>}
  n51 = StaticSwitchParameter(True=NamedRerouteUsage(), False=NamedRerouteUsage()) {parameterName=UseBranchLevel}
  NamedRerouteDeclaration(None=@n51) {name=Leaf_Level_Pivot, nodeColor=<rgba 0.6875 0 1 1>}
  n52 = StaticSwitchParameter(True=NamedRerouteUsage(), False=NamedRerouteUsage()) {parameterName=UseBranchLevel}
  NamedRerouteDeclaration(None=@n52) {name=Branch_Level_XAxis, nodeColor=<rgba 0.2891 0 1 1>}
  n53 = StaticSwitchParameter(True=NamedRerouteUsage(), False=NamedRerouteUsage()) {parameterName=UseBranchLevel}
  NamedRerouteDeclaration(None=@n53) {name=Leaf_Level_AXis, nodeColor=<rgba 0.5703 1 0 1>}
  n54 = ComponentMask(None=@n27) {g=true}
  NamedRerouteDeclaration(None=@n54) {name=TextureSizeY, nodeColor=<rgba 1 0.5156 0 1>}
  n55 = Multiply(A=Time(), B=ScalarParameter() {defaultValue=0.5, group=Anim, parameterName=WindGustsSpeed})
  n56 = Divide(A=Add(A=Sine(None=@n55)))
  n57 = ScalarParameter() {defaultValue=1, group=Anim, parameterName=MaxWindGusts}
  n58 = MaterialFunctionCall() {materialFunction=/Game/Athena/Environments/Landscape/MaterialFunctions/MF_StormMask}
  n59 = Max(
   A=Multiply(
    A=MaterialFunctionCall!(FadeOutTIme, ResetFadeInTime, FadeInTime) {materialFunction=/Game/Environments/TechArt/WindGustSystem/MF_WindGustsSystem}.FadeInFadeOutNormalized
    B=Multiply(A=@n56, B=@n57)
   )
   B=@n58.Result
  )
  n60 = CollectionParameter() {
   collection=&FortniteMaterialParameters
   parameterName=WeatherWindDirectionAndStrength
  }
  n61 = Saturate(None=Add(A=QualitySwitch(Default=@n59, Low=@n58.Result), B=ComponentMask(None=@n60) {a=true}))
  n62 = WorldPosition() {worldPositionShaderOffset=WPT_ExcludeAllShaderOffsets}
  n63 = Add(A=ShadingPathSwitch(Default=SamplePhysicsVectorField!(World Position=@n62.XYZ) {fieldTarget=Vector_LinearVelocity}, Forward=@n62.XYZ, Mobile=@n62.XYZ) {desc=Need this to prevent compile error}) {constB=0.0001}
  n64 = MaterialFunctionCall(Vector 3=@n63) {materialFunction=/Engine/Functions/Engine_MaterialFunctions02/Utility/VectorLength}
  n65 = MaterialFunctionCall!(InVector=Multiply(A=ScalarParameter() {defaultValue=1, group=GlobalForces, parameterName=GlobalForceStormAddToStormMask}, B=@n64.V3 Length)) {materialFunction=/Game/Athena/Environments/Landscape/Foliage/MF_MakePrecise}
  n66 = QualitySwitch(Default=@n61, Epic=Saturate(None=Add(A=@n65.Result, B=@n61)))
  NamedRerouteDeclaration(None=StaticSwitchParameter(True=NamedRerouteUsage(), False=ShadingPathSwitch(Default=@n66, Forward=@n61, Mobile=@n61) {desc=Need this to prevent compile error}) {group=Anim, parameterName=DebugWindSpeed}) {name=Storm_Mask, nodeColor=<rgba 0 0.2734 1 1>}
  n67 = Time()
  n68 = ScalarParameter() {defaultValue=0.2, parameterName=DebugSpeed}
  MaterialExpressionRemap_0 = MaterialXRemap!(None=Sine(None=Multiply(A=@n67, B=@n68)), InputLow, InputHigh, TargetLow, TargetHigh) {inputLowDefault=-1, targetHighDefault=2, targetLowDefault=-0.5}
  NamedRerouteDeclaration(None=Saturate(None=@MaterialExpressionRemap_0)) {name=DebugWind, nodeColor=<rgba 0 1 0.5547 1>}
  n69 = CollectionParameter() {
   collection=&FortniteMaterialParameters
   parameterName=WindDirection_RGB Speed_A
  }
  n70 = Normalize(VectorInput=Multiply(A=ComponentMask(None=@n69) {b=true, g=true, r=true}) {constB=-1})
  n71 = Constant3Vector() {constant=<rgba 1 1 0 0>}
  n72 = Multiply(A=Normalize(VectorInput=Multiply(A=@n63, B=@n71)), B=Constant3Vector() {constant=<rgba -1 -1 0 0>}) {constB=-1}
  n73 = ScalarParameter() {defaultValue=3, group=GlobalForces, parameterName=GlobalForceDirectionIntensity}
  n74 = MaterialFunctionCall!(InVector=Multiply(A=Multiply(A=@n72, B=@n73), B=@n64.V3 Length)) {materialFunction=/Game/Athena/Environments/Landscape/Foliage/MF_MakePreciseVec3}
  n75 = QualitySwitch(Default=@n70, Epic=Normalize(VectorInput=Add(A=@n70, B=@n74.Result)))
  n76 = ShadingPathSwitch(Default=@n75, Forward=@n75, Mobile=@n75) {desc=Need this to prevent compile error}
  NamedRerouteDeclaration(None=@n76) {name=WindDirection, nodeColor=<rgba 1 0 0.4687 1>}
  n77 = NamedRerouteUsage()
  n78 = Normalize(VectorInput=CrossProduct(A=NamedRerouteUsage(), B=ComponentMask(None=@n77) {b=true, g=true, r=true}))
  NamedRerouteDeclaration(None=@n78) {name=Rotation_Axis_Leaves, nodeColor=<rgba 1 0.7734 0 1>}
  n79 = MaterialFunctionCall() {materialFunction=&LocalPosition}
  n80 = VectorParameter() {group=Anim_Advanced_Tweak, parameterName=RandomnessPositionFromBP, primitiveDataIndex=29, defaultValue=<rgba 0 0 0 1>}
  n81 = StaticBoolParameter() {defaultValue=true, group=Anim_Advanced_Tweak, parameterName=RandomnessFromWorldSpace}
  n82 = StaticSwitch(
   True=WorldPosition().XYZ
   False=Abs(None=Add(A=@n79.Local Position, B=@n80.RGB))
   Value=NamedRerouteDeclaration(None=@n81) {name=RandomFromWSbool, nodeColor=<rgba 0.8047 1 0 1>}
  )
  NamedRerouteDeclaration(None=@n82) {name=WorldPosition_Or_BPposition, nodeColor=<rgba 0 0.0859 1 1>}
  n83 = ScalarParameter() {defaultValue=0.075, group=Anim, parameterName=RotationAngle_Leafs}
  n84 = ScalarParameter() {defaultValue=2, group=Anim, parameterName=RotationAngle_Leafs_STORM_Mulitply}
  n85 = QualitySwitch(Default=@n84, Epic=Add(A=NamedRerouteUsage(), B=@n84))
  n86 = LinearInterpolate(
   A=@n83
   B=Multiply(A=@n83, B=ShadingPathSwitch(Default=@n85, Forward=@n84, Mobile=@n84) {desc=Need this to prevent compile error})
   Alpha=NamedRerouteUsage()
  )
  NamedRerouteDeclaration(None=@n86) {name=Rotation_Amount_Leaves, nodeColor=<rgba 0 1 0.4141 1>}
  n87 = NamedRerouteUsage()
  n88 = NamedRerouteUsage()
  n89 = NamedRerouteUsage()
  n90 = Saturate(None=Divide(A=Distance(A=@n87, B=@n88), B=ComponentMask(None=@n89) {a=true}))
  NamedRerouteDeclaration(None=@n90) {name=Gradient_Along_Leaf, nodeColor=<rgba 1 0 0.3984 1>}
  n91 = NamedRerouteUsage()
  n92 = Subtract(A=DotProduct(A=ComponentMask(None=@n91) {b=true, g=true, r=true}, B=NamedRerouteUsage()))
  n93 = Multiply(A=ScalarParameter() {defaultValue=-0.25, group=Anim, parameterName=RotationAngle_Leafs_Windfacing_Bias}, B=OneMinus(None=Divide(A=@n92)))
  NamedRerouteDeclaration(None=@n93) {name=RotationAxis_Leaves, nodeColor=<rgba 0.8516 1 0 1>}
  n94 = TextureSampleParameter2D(UVs=@n41) {
   parameterName=Position
   samplerSource=SSM_Wrap_WorldGroupSettings
   samplerType=SAMPLERTYPE_LinearColor
   texture=&Position
  }
  n95 = VectorParameter() {group=Anim_Advanced_Tweak, parameterName=RandomnessPositionFromBP, primitiveDataIndex=29, defaultValue=<rgba 0 0 0 1>}
  n96 = StaticSwitch(
   True=TransformPosition(None=@n94.RGB) {transformType=TRANSFORMPOSSOURCE_World}
   False=Abs(None=Add(A=@n94.RGB, B=@n95.RGB))
   Value=NamedRerouteUsage()
  )
  NamedRerouteDeclaration(None=@n96) {name=Pivot_Level_2, nodeColor=<rgba 0 1 0.2266 1>}
  n97 = Saturate(None=@n65.Result)
  NamedRerouteDeclaration(None=@n97) {name=PhysicsForceAdd, nodeColor=<rgba 0 1 0.3203 1>}
  n98 = MaterialFunctionCall(InVector=Multiply(A=@n64.V3 Length, B=ScalarParameter() {defaultValue=0.5, group=GlobalForces, parameterName=GlobalForceBendAngleAddition})) {materialFunction=&MF_MakePrecise}
  NamedRerouteDeclaration(None=@n98.Result) {name=PhysicsForceAngleAdd, nodeColor=<rgba 1 0.7734 0 1>}
  n99 = NamedRerouteUsage()
  n100 = ComponentMask(None=NamedRerouteUsage()) {b=true, g=true, r=true}
  n101 = NamedRerouteUsage()
  n102 = NamedRerouteUsage()
  n103 = NamedRerouteUsage()
  n104 = Saturate(None=Divide(A=Distance(A=@n101, B=@n102), B=ComponentMask(None=@n103) {a=true}))
  n105 = ScalarParameter() {defaultValue=4, group=Anim, parameterName=RotationAngle_TrunkAnimExponent}
  n106 = Transform(None=Constant3Vector() {constant=<rgba 1 1 0 0>}) {transformSourceType=TRANSFORMSOURCE_Local}
  n107 = Frac(None=Add(A=ComponentMask(None=@n106) {r=true}, B=ComponentMask(None=@n106) {g=true}))
  n108 = ScalarParameter() {defaultValue=33.232, group=Anim_Advanced_Tweak, parameterName=RandomnessTrunk}
  n109 = VectorParameter() {group=Anim_Advanced_Tweak, parameterName=RandomnessPositionFromBP, primitiveDataIndex=29, defaultValue=<rgba 0 0 0 1>}
  n110 = Frac(None=Add(A=ComponentMask(None=@n109.RGB) {r=true}, B=ComponentMask(None=@n109.RGB) {g=true}))
  n111 = Add(
   A=Multiply(A=NamedRerouteUsage(), B=ScalarParameter() {defaultValue=0.05, group=Anim_Speed_Control, parameterName=Speed_Trunk})
   B=StaticSwitch(True=Multiply(A=@n107, B=@n108), False=Multiply(A=@n110, B=@n108), Value=NamedRerouteUsage())
  )
  n112 = TextureSample(UVs=Frac(None=NamedRerouteDeclaration(None=@n111) {name=TrunkAnimSpeed, nodeColor=<rgba 0 1 0.3203 1>})) {
   samplerSource=SSM_Wrap_WorldGroupSettings
   samplerType=SAMPLERTYPE_LinearGrayscale
   texture=/Game/Environments/Apollo/Foliage/Wind/Wind_SineWaves_HDR
  }
  n113 = Add(
   A=LinearInterpolate(
    A=Constant() {desc=Fixed bias to -1 for gameplay to have minimum amount of motion from the collision, r=-0.8}
    B=Constant() {desc=Fixed bias to -1 for gameplay to have minimum amount of motion from the collision, r=-0.5}
    Alpha=NamedRerouteUsage()
   )
   B=@n112.R
  )
  n114 = ScalarParameter() {defaultValue=0.03, group=Anim, parameterName=RotationAngle_Trunk}
  n115 = LinearInterpolate(
   A=@n114
   B=Multiply(A=@n114, B=ScalarParameter() {defaultValue=2, group=Anim, parameterName=RotationAngle_Trunk_STORM_MULTIPLY})
   Alpha=NamedRerouteUsage()
  )
  n116 = NamedRerouteDeclaration(None=@n115) {name=Rotation_Amount_Trunk, nodeColor=<rgba 0 1 0.0859 1>}
  n117 = RotateAboutAxis(
   NormalizedRotationAxis=Normalize(VectorInput=CrossProduct(A=@n99, B=@n100))
   RotationAngle=Multiply(A=Power(Base=@n104, Exp=@n105), B=Multiply(A=@n113, B=@n116))
   PivotPoint=NamedRerouteUsage()
   Position=NamedRerouteUsage()
  )
  NamedRerouteDeclaration(None=@n117) {name=FinalAnim_Trunk, nodeColor=<rgba 0.7813 0 1 1>}
  n118 = NamedRerouteUsage()
  n119 = ComponentMask(None=NamedRerouteUsage()) {b=true, g=true, r=true}
  n120 = Distance(A=NamedRerouteUsage(), B=NamedRerouteUsage())
  n121 = ComponentMask(None=NamedRerouteUsage()) {a=true}
  n122 = Power(
   Base=Saturate(None=Divide(A=@n120, B=@n121))
   Exp=ScalarParameter() {defaultValue=4, group=Anim, parameterName=RotationAngle_TrunkAnimExponent}
  )
  n123 = Multiply(A=NamedRerouteUsage(), B=ScalarParameter() {defaultValue=0.03, group=Anim_Speed_Control, parameterName=Speed_Branches})
  n124 = ComponentMask(None=NamedRerouteUsage()) {b=true, g=true, r=true}
  n125 = ComponentMask(None=@n124) {b=true}
  n126 = ComponentMask(None=@n124) {g=true}
  n127 = Multiply(A=Frac(None=Add(A=@n125, B=@n126)), B=ScalarParameter() {defaultValue=33.8762, group=Anim_Advanced_Tweak, parameterName=RandomnessBranches})
  n128 = TextureSample(UVs=Frac(None=Add(A=@n123, B=@n127))) {
   samplerSource=SSM_Wrap_WorldGroupSettings
   samplerType=SAMPLERTYPE_LinearGrayscale
   texture=&Wind_SineWaves_HDR
  }
  n129 = ScalarParameter() {
   defaultValue=-0.25
   desc="-1 is equal motion to both sides. -2 is only motion away from Wind.0 is only motion into the wind (don't do that)"
   group=Anim
   parameterName=RotationAngle_Branch_Bias
  }
  n130 = ScalarParameter() {defaultValue=0.05, group=Anim, parameterName=RotationAngle_Branches}
  n131 = ScalarParameter() {defaultValue=2, group=Anim, parameterName=RotationAngle_Branches_STORM_MULTIPY}
  n132 = Multiply(A=Add(A=@n128.R, B=@n129), B=LinearInterpolate(A=@n130, B=Multiply(A=@n130, B=@n131), Alpha=NamedRerouteUsage())) {constB=4}
  n133 = RotateAboutAxis(
   NormalizedRotationAxis=Normalize(VectorInput=CrossProduct(A=@n118, B=@n119))
   RotationAngle=StaticSwitchParameter(True=Multiply(A=@n122, B=@n132), False=@n132) {parameterName=UseBranchMotionFalloff}
   PivotPoint=NamedRerouteUsage()
   Position=NamedRerouteUsage()
  )
  NamedRerouteDeclaration(None=@n133) {name=FinalAnim_Branches, nodeColor=<rgba 0.125 1 0 1>}
  n134 = Multiply(A=NamedRerouteUsage(), B=ScalarParameter() {defaultValue=0.07, group=Anim_Speed_Control, parameterName=Speed_leafs})
  n135 = Frac(None=@n2)
  n136 = ScalarParameter() {defaultValue=33.0019, group=Anim_Advanced_Tweak, parameterName=Randomness_Leafs}
  n137 = Add(
   A=Add(A=NamedRerouteDeclaration(None=@n134) {name=SpeedLeaves, nodeColor=<rgba 1 0.9375 0 1>}, B=Multiply(A=@n135, B=@n136))
   B=Multiply(A=ScalarParameter() {defaultValue=0.0025, group=Anim, parameterName=Drag_Leafs}, B=NamedRerouteUsage())
  )
  n138 = TextureSample(UVs=Frac(None=@n137)) {
   samplerSource=SSM_Wrap_WorldGroupSettings
   samplerType=SAMPLERTYPE_LinearGrayscale
   texture=&Wind_SineWaves_HDR
  }
  n139 = ScalarParameter() {defaultValue=10, group=GlobalForces, parameterName=ForceWindSpeed}
  n140 = TextureSample(UVs=Frac(None=Multiply(A=@n137, B=@n139) {constB=2})) {
   samplerSource=SSM_Wrap_WorldGroupSettings
   samplerType=SAMPLERTYPE_LinearGrayscale
   texture=&Wind_SineWaves_HDR
  }
  n141 = Add(A=@n140.R) {constB=-0.5}
  n142 = NamedRerouteUsage()
  n143 = Add(
   A=@n138.R
   B=Multiply(A=Multiply(A=@n141, B=@n142), B=ScalarParameter() {defaultValue=0.2, group=GlobalForces, parameterName=ForceWindIntensity})
  )
  n144 = Add(
   A=ShadingPathSwitch(Default=QualitySwitch(Default=@n138.R, Epic=@n143), Forward=@n138.R, Mobile=@n138.R) {desc=Need this to prevent compile error}
   B=NamedRerouteUsage()
  )
  n145 = NamedRerouteUsage()
  n146 = RotateAboutAxis(
   NormalizedRotationAxis=NamedRerouteUsage()
   RotationAngle=Multiply(A=Multiply(A=@n144, B=@n145) {constB=4}, B=NamedRerouteUsage())
   PivotPoint=NamedRerouteUsage()
   Position=NamedRerouteUsage()
  )
  NamedRerouteDeclaration(None=@n146) {name=FinalAnim_Leaves, nodeColor=<rgba 1 0 0.6562 1>}
  n147 = Constant3Vector() {constant=<rgba 0 0 1 0>}
  n148 = NamedRerouteUsage()
  n149 = Multiply(A=NamedRerouteUsage(), B=Constant() {r=4})
  n150 = NamedRerouteUsage()
  n151 = TransformPosition(None=Constant3Vector()) {transformType=TRANSFORMPOSSOURCE_World}
  n152 = Saturate(None=Divide(A=Distance(A=@n150, B=@n151), B=ScalarParameter() {defaultValue=2000, group=SimpleAnimation, parameterName=SimpleAnimation_ActorHeight}))
  n153 = NamedRerouteUsage()
  n154 = RotateAboutAxis(
   NormalizedRotationAxis=Normalize(VectorInput=CrossProduct(A=@n147, B=@n148))
   RotationAngle=Multiply(A=Sine(None=@n149), B=Multiply(A=@n152, B=@n153))
   PivotPoint=@n151
   Position=NamedRerouteUsage()
  )
  n155 = Constant3Vector()
  n156 = Multiply(A=ScalarParameter() {defaultValue=800, group=SimpleAnimation, parameterName=SimpleAnimation_LeafSpeed}, B=NamedRerouteUsage())
  n157 = NamedRerouteUsage()
  n158 = Frac(None=Divide(A=Add(A=@n156, B=@n157), B=ScalarParameter() {defaultValue=40, group=SimpleAnimation, parameterName=WaveLengthSimpleAnimation}))
  n159 = NamedRerouteUsage()
  n160 = ScalarParameter() {defaultValue=100, group=SimpleAnimation, parameterName=SimpleAnimation_LeafsMovement}
  n161 = LinearInterpolate(
   A=@n160
   B=Multiply(A=@n160, B=ScalarParameter() {defaultValue=1.2, group=SimpleAnimation, parameterName=SimpleAnimation_LeafsMovement_STORMMult})
   Alpha=NamedRerouteUsage()
  )
  n162 = Multiply(A=Multiply(A=Sine(None=@n158), B=@n152), B=Multiply(A=NamedRerouteUsage(), B=Multiply(A=@n159, B=@n161)))
  n163 = Add(A=@n162, B=@n154)
  n164 = StaticSwitchParameter(
   True=ShadingPathSwitch(Default=@n154, Deferred=@n154, Forward=@n155, Mobile=@n155) {desc=Only leaf motion on moblie and switch}
   False=ShadingPathSwitch(Default=@n163, Deferred=@n163, Forward=@n162, Mobile=@n162) {desc=Only leaf motion on mobile and switch}
  ) {group=Trunk, parameterName=IsTrunk}
  n165 = StaticSwitchParameter(True=Multiply(A=VertexColor().R, B=@n164), False=@n164) {parameterName=UseVertexcolormaskforSImpleAnimation}
  NamedRerouteDeclaration(None=@n165) {name=FinalSimpleANim, nodeColor=<rgba 0 1 0.7188 1>}
  n166 = TextureSampleParameter2D(UVs=@n35) {
   parameterName=Position
   samplerSource=SSM_Wrap_WorldGroupSettings
   samplerType=SAMPLERTYPE_LinearColor
   texture=&Position
  }
  n167 = VectorParameter() {group=Anim_Advanced_Tweak, parameterName=RandomnessPositionFromBP, primitiveDataIndex=29, defaultValue=<rgba 0 0 0 1>}
  n168 = StaticSwitch(
   True=TransformPosition(None=@n166.RGB) {transformType=TRANSFORMPOSSOURCE_World}
   False=Abs(None=Add(A=@n166.RGB, B=@n167.RGB))
   Value=NamedRerouteUsage()
  )
  NamedRerouteDeclaration(None=@n168) {name=Pivot_level_1, nodeColor=<rgba 1 0.9375 0 1>}
  n169 = VectorParameter() {group=Anim_Advanced_Tweak, parameterName=RandomnessPositionFromBP, primitiveDataIndex=29, defaultValue=<rgba 0 0 0 1>}
  n170 = StaticSwitch(
   True=TransformPosition(None=@n32.RGB) {transformType=TRANSFORMPOSSOURCE_World}
   False=Abs(None=Add(A=@n32.RGB, B=@n169.RGB))
   Value=NamedRerouteUsage()
  )
  NamedRerouteDeclaration(None=@n170) {name=Pivot_Level_3, nodeColor=<rgba 1 0 0.6328 1>}
  NamedRerouteDeclaration(None=@n166.A) {name=Mask_Level_1, nodeColor=<rgba 0 0.2969 1 1>}
  n171 = Clamp(None=NamedRerouteUsage())
  NamedRerouteDeclaration(None=@n171) {name=Mask_Trunk, nodeColor=<rgba 1 0.3516 0 1>}
  NamedRerouteDeclaration(None=@n94.A) {name=Mask_Level_2, nodeColor=<rgba 0.8281 1 0 1>}
  n172 = Clamp(None=StaticSwitchParameter(True=NamedRerouteUsage(), False=NamedRerouteUsage()) {parameterName=UseBranchLevel})
  NamedRerouteDeclaration(None=@n172) {name=Mask_Leaves, nodeColor=<rgba 0 0.6484 1 1>}
  Reroute(None=@n52)
  Reroute(None=@n51)
  Reroute(None=@n50)
  Reroute(None=@n53) {desc=X Axis Level2}
  Reroute(None=@n28)
  Reroute(None=@n81)
  Reroute(None=@n54)
  Reroute(None=@n96)
  Reroute(None=@n168)
  Reroute(None=@n170)
  Reroute(None=@n166.A)
  Reroute(None=@n171)
  Reroute(None=@n94.A)
  Reroute(None=@n172)
  Reroute(None=@n32.A)
  Reroute(None=@n38)
  Reroute(None=@n44)
  Reroute(None=@n47)
  Reroute(None=@n48)
  Reroute(None=@n49)
  Reroute(None=@n76)
  Reroute(None=@n78)
  Reroute(None=@n82)
  Reroute(None=@n86)
  Reroute(None=@n90)
  Reroute(None=@n115)
  Reroute(None=@n93)
  Reroute(None=@n97)
  Reroute(None=@n98.Result)
  Reroute(None=@n111)
  Reroute(None=@n117)
  Reroute(None=@n133)
  Reroute(None=@n146)
  Reroute(None=@n165)
 ]}
