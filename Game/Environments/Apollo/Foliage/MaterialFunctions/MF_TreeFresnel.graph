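# Material function that adds view-angle fresnel and top-down highlights to tree foliage base color, only at Epic quality outside mobile shading.
# A two-sided vertex normal blended with the transformed detail normal, or the input normal for impostors, drives both terms.
# Fresnel and tinted up-facing power lobes are added to BaseColor, darkened, then faded by a mask weighted by SelfShadowFresnel.
# Other quality levels and mobile pass attributes through unchanged.
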MaterialFunction /Game/Environments/Apollo/Foliage/MaterialFunctions/MF_TreeFresnel
graph {nodes [
  Constant()
  n1 = FunctionInput() {inputType=FunctionInput_MaterialAttributes}
  n2 = GetMaterialAttributes(None=@n1) {attributeGetTypes [#1, #2]}
  n3 = ScalarParameter() {defaultValue=0.1, group=EpicDetail, parameterName=FresnelEpic_Power3b_Brightness}
  n4 = Multiply(A=TwoSidedSign(), B=VertexNormalWS())
  n5 = Transform(None=@n2.Normal)
  n6 = ScalarParameter() {defaultValue=0.5, group=EpicDetail, parameterName=normaldetail}
  n7 = StaticSwitch(
   True=Normalize(VectorInput=@n2.Normal)
   False=Normalize(VectorInput=LinearInterpolate(A=@n4, B=@n5, Alpha=@n6))
   Value=FunctionInput(Preview=StaticBool()) {bUsePreviewValueAsDefault=true, inputName=IsImpostor, inputType=FunctionInput_StaticBool, sortPriority=4}
  )
  n8 = DotProduct(A=CameraVectorWS(), B=@n7)
  n9 = Clamp(None=Power(Base=OneMinus(None=@n8), Exp=ScalarParameter() {defaultValue=3, group=EpicDetail, parameterName=FresnelEpic_Power3b}))
  n10 = ScalarParameter() {defaultValue=0.05, group=EpicDetail, parameterName=FresnelEpic_ColorTopDown_Brightness}
  n11 = DotProduct(A=Constant3Vector() {constant=<rgba 0 0 1 0>}, B=@n7)
  n12 = ScalarParameter() {defaultValue=70, group=EpicDetail, parameterName=FresnelEpic_Power2b}
  n13 = Add(
   A=Multiply(A=ScalarParameter() {defaultValue=0.8, group=EpicDetail, parameterName=FresnelEpic_Power2b_Brightness}, B=Power(Base=@n11, Exp=@n12))
   B=Power(Base=@n11, Exp=ScalarParameter() {defaultValue=6, group=EpicDetail, parameterName=FresnelEpic_Power2})
  )
  n14 = Multiply(
   A=Multiply(A=@n2.BaseColor, B=@n13)
   B=VectorParameter() {group=EpicDetail, parameterName=BaseColorTopDownFresnel, defaultValue=<rgba 0.8776 1 0.1856 1>}.RGB
  )
  n15 = Add(
   A=Multiply(A=ScalarParameter() {defaultValue=0.6, group=EpicDetail, parameterName=FresnelEpic_ColorSun_Brightness}, B=Multiply(A=@n3, B=@n9))
   B=Add(A=Multiply(A=@n10, B=@n13), B=Add(A=@n2.BaseColor, B=@n14))
  )
  n16 = ScalarParameter() {defaultValue=1, group=EpicDetail, parameterName=FresnellOverallDarkenAll}
  n17 = Constant() {r=1}
  n18 = FunctionInput() {inputName=Mask, inputType=FunctionInput_Scalar, sortPriority=2}
  n19 = ScalarParameter() {defaultValue=0.5, group=EpicDetail, parameterName=SelfShadowFresnel}
  n20 = SetMaterialAttributes!(
   MaterialAttributes=@n2.MaterialAttributes
   EmissiveColor
   BaseColor=LinearInterpolate(A=@n2.BaseColor, B=Multiply(A=@n15, B=@n16), Alpha=LinearInterpolate(A=@n17, B=@n18, Alpha=@n19))
  ) {attributeSetTypes [#3, #2]}
  FunctionOutput(None=ShadingPathSwitch(Default=QualitySwitch(Default=@n1, Epic=@n20), Mobile=@n1) {desc=Migrated from FeatureLevelSwitch})
  VectorParameter() {parameterName=ColorTopDownFresnel, defaultValue=<rgba 0.0172 0.032 0.05 1>}
  VectorParameter() {parameterName=ColorOpositeSunFresnel, defaultValue=<rgba 0.6682 0.7031 0.2576 1>}
 ]}
