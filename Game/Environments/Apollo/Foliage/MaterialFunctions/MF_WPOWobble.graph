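# Foliage material function that adds a temporary hit-reaction wobble to World Position Offset on incoming Material Attributes.
# A time-driven sine along WobbleDirection, scaled by WobbleStrength, is masked by XY-only distance falloff around HitLocation.
# It fades linearly over WobbleSeconds from the timestamp in the Time parameter, then adds to the existing offset.
# The "Use WPO Wobble" static switch passes the input through unchanged when off.
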MaterialFunction /Game/Environments/Apollo/Foliage/MaterialFunctions/MF_WPOWobble
graph {nodes [
  n1 = FunctionInput() {inputType=FunctionInput_MaterialAttributes}
  n2 = GetMaterialAttributes(None=@n1) {attributeGetTypes [#1]}
  n3 = Constant3Vector() {constant=<rgba 1 1 0 0>}
  n4 = Distance(
   A=Multiply(A=WorldPosition().XYZ, B=@n3)
   B=Multiply(A=@n3, B=VectorParameter() {parameterName=HitLocation, defaultValue=<rgba 0 0 0 1>}.RGB)
  )
  n5 = ScalarParameter() {defaultValue=200, parameterName=WobbleDistance}
  n6 = OneMinus(None=Saturate(None=Divide(A=@n4, B=@n5)))
  n7 = Multiply(A=Time(), B=ScalarParameter() {defaultValue=3, parameterName=WobbleSpeed})
  n8 = Multiply(
   A=Multiply(A=Sine(None=@n7), B=VectorParameter() {parameterName=WobbleDirection, defaultValue=<rgba 1 0 0 1>}.RGB)
   B=ScalarParameter() {defaultValue=30, parameterName=WobbleStrength}
  )
  n9 = Time()
  n10 = ScalarParameter() {parameterName=Time}
  n11 = Saturate(None=Divide(A=Subtract(A=@n9, B=@n10), B=ScalarParameter() {defaultValue=1.5, parameterName=WobbleSeconds}))
  n12 = Add(A=Multiply(A=Multiply(A=@n6, B=@n8), B=OneMinus(None=@n11)), B=@n2.WorldPositionOffset)
  FunctionOutput(None=StaticSwitchParameter(
    True=SetMaterialAttributes!(MaterialAttributes=@n2.MaterialAttributes, WorldPositionOffset=@n12) {attributeSetTypes [#1]}
    False=@n1
   ) {parameterName=Use WPO Wobble})
 ]}
